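# Stereo MetaSound source for the foley of the StopWatch emote, playing a start sound followed by a repeating ticking sound.
# On play, it picks a random start wave and pitch-shifts it via a frequency multiplier converted to semitones.
# When that finishes, a TriggerRepeat periodically retriggers the unpitched ticking wave.
# A two-channel mixer applies separate intro and loop gains; several inputs and variables are unused.
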MetaSoundSource /BRCosmetics/Sounds/Emotes/StopWatch/MSS_Emote_Foley_StopWatch
inputs [
 DelayLoopByStart {type=Bool, access=Constructor}
 FadeInTime {type=Time, access=Constructor}
 Foley.Loop.Sounds {type=WaveAsset:Array, access=Constructor}
 Foley.Sounds {type=WaveAsset:Array, access=Constructor}
 IntroVolumeMultiplier {type=Float}
 Loop.DelayTime {type=Time}
 Loop.FadeCurveExp {type=Float}
 Loop.FadeTimeIn {type=Float}
 Looping {type=Bool, access=Constructor}
 LoopVolumeMultiplier {type=Float}
 PitchTemplateMultiplier {type=Float}
 Play {type=Trigger}
 Start {type=Trigger}
 TestFloat {type=Float}
 TickingLoop {type=WaveAsset}
 Trigger {type=Trigger}
 UE.Source.OnPlay {type=Trigger}
]
interfaces [UE.OutputFormat.Stereo:1.0, UE.Source:1.1]
outputs [
 "UE.OutputFormat.Stereo.Audio:0" {type=Audio}
 "UE.OutputFormat.Stereo.Audio:1" {type=Audio}
]
pages [Default {
  input_defaults {
   DelayLoopByStart="<MetaSoundBuilderResult.SUCCEEDED: 0>"
   FadeInTime=0
   IntroVolumeMultiplier=1
   Loop.DelayTime=0
   Loop.FadeCurveExp=1
   Loop.FadeTimeIn=0
   LoopVolumeMultiplier=1
   Looping="<MetaSoundBuilderResult.SUCCEEDED: 0>"
   PitchTemplateMultiplier=1.032
   Play=null
   Start=null
   TestFloat=40.4
   TickingLoop=&/Emote_StopWatch_Loop_01
   Trigger=null
   UE.Source.OnPlay=null
   Foley.Loop.Sounds [&Emote_StopWatch_Loop_01]
   Foley.Sounds [
    &/Emote_StopWatch_Start_03
    &/Emote_StopWatch_Start_02
    &/Emote_StopWatch_Start_01
   ]
  }
  nodes [
   n1 = InitVariable.Audio()
   ArrayRandomGet_0 = Array.Random Get:WaveAsset:Array(In Array=^Foley.Sounds, Next=^UE.Source.OnPlay) {also {defaults {Weights []}}}
   FrequencyMultipliertoSemitone_0 = UE.Frequency Multiplier to Semitone:Float(Frequency Multiplier=^PitchTemplateMultiplier)
   WavePlayer_0 = UE.Wave Player:Stereo(Play=@ArrayRandomGet_0.On Next, Wave Asset=@ArrayRandomGet_0.Value, Pitch Shift=@FrequencyMultipliertoSemitone_0.Semitones)
   n2 = InitVariable.Audio()
   n3 = VariableMutator.Trigger(Variable=InitVariable.Trigger().Variable, Value=@WavePlayer_0.On Finished)
   WavePlayer_1 = UE.Wave Player:Stereo(
    Wave Asset=^TickingLoop
    Play=UE.TriggerRepeat(Start=VariableAccessor.Trigger(Variable=@n3.Variable).Value) {also {defaults {Period=4}}}.RepeatOut
   )
   n4 = InitVariable.Audio()
   n5 = InitVariable.Audio()
   AudioMixerAudioMixerStereo2_0 = "AudioMixer.Audio Mixer (Stereo, 2)"(
    Gain 0=^IntroVolumeMultiplier
    Gain 1=^LoopVolumeMultiplier
    In 0 L=VariableAccessor.Audio(Variable=VariableMutator.Audio(Variable=@n1.Variable, Value=@WavePlayer_0.Out Left).Variable).Value
    In 1 L=VariableAccessor.Audio(Variable=VariableMutator.Audio(Variable=@n2.Variable, Value=@WavePlayer_1.Out Left).Variable).Value
    In 1 R=VariableAccessor.Audio(Variable=VariableMutator.Audio(Variable=@n4.Variable, Value=@WavePlayer_1.Out Right).Variable).Value
    In 0 R=VariableAccessor.Audio(Variable=VariableMutator.Audio(Variable=@n5.Variable, Value=@WavePlayer_0.Out Right).Variable).Value
   )
   InitVariable.Audio()
   InitVariable.Audio()
   InitVariable.Audio()
   InitVariable.Audio()
   InitVariable.Float()
   InitVariable.Float()
   InitVariable.Float()
   InitVariable.Float()
   InitVariable.Float()
   InitVariable.Float()
   InitVariable.Time()
   InitVariable.Trigger()
   InitVariable.Trigger()
   InitVariable.Trigger()
   InitVariable.Trigger()
   InitVariable.Trigger()
   InitVariable.Trigger()
   VariableMutator.Trigger(Variable=InitVariable.Trigger().Variable)
  ]
  links [
   {dst="@UE.OutputFormat.Stereo.Audio:0", src=AudioMixerAudioMixerStereo2_0.Out L}
   {dst="@UE.OutputFormat.Stereo.Audio:1", src=AudioMixerAudioMixerStereo2_0.Out R}
  ]
 }]
properties {
 assetImportData=&self:AssetImportData
 attenuationSettings=&Fort_Emotes_Athena_Att
 audioPropertiesSheet=/BRCosmetics/Sounds/Settings/PropertySheets/Emotes/APS_Cosmetic_Emote_Foley
 bLooping=true
 duration=10000
 maxDistance=1600
 numChannels=2
 outputFormat=Stereo
 sampleRate=48000
 soundClassObject=/Game/Sounds/FortSoundClassesAndModes/Fort_Foley_Emote
 virtualizationMode=PlayWhenSilent
 concurrencyOverrides {maxCount=3}
 concurrencySet [/Game/Sounds/Concurrency_Rules/Athena_EmoteFoley_Conc]
 soundSubmixSends [{}]
}
